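; Burst gameplay cue notify that plays the siphon healing sound for the Double Trouble shotgun.
; On burst, it checks that the target is valid and picks the high-magnitude sound when the cue's normalized magnitude is at least 2, otherwise the low one.
; A locally controlled instigator of the target hears it as 2D audio; otherwise the sound attaches to that instigator's root component.
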(fn OnBurst(EitherVFX-preferthisversion)
  (params (MyTarget :type Actor%20Object%20Reference) (Parameters :type Gameplay%20Cue%20Parameters%20Structure%20%28by%20ref%29) (ParticleComponents :type Array%20of%20FXSystem%20Component%20Object%20References) (AudioComponents :type Array%20of%20Audio%20Component%20Object%20References) (BurstCameraShakeInstance :type Camera%20Shake%20Base%20Object%20Reference) (BurstDecalInstance :type Decal%20Actor%20Object%20Reference))
  (bind _normalizedmagnitude (Ability|GameplayCue|BreakGameplayCueParameters Parameters))
  (bind _returnvalue (>= _normalizedmagnitude 2.0))
  (bind _instigator (Variables|/Script/Engine.Actor|GetInstigator MyTarget))
  (Utilities|IsValid MyTarget
    (:Is%20Valid
      (if (Pawn|IsLocallyControlled _instigator)
        (Audio|SpawnSound2D 0 (select _returnvalue "/TidalCraneExotics/Gameplay/DoubleTrouble/Audio/SC_Shotgun_Siphon_Healing_HighMag.SC_Shotgun_Siphon_Healing_HighMag" "/TidalCraneExotics/Gameplay/DoubleTrouble/Audio/SC_Shotgun_Siphon_Healing_LowMag.SC_Shotgun_Siphon_Healing_LowMag"))
        (else
          (Audio|SpawnSoundAttached (select _returnvalue "/TidalCraneExotics/Gameplay/DoubleTrouble/Audio/SC_Shotgun_Siphon_Healing_HighMag.SC_Shotgun_Siphon_Healing_HighMag" "/TidalCraneExotics/Gameplay/DoubleTrouble/Audio/SC_Shotgun_Siphon_Healing_LowMag.SC_Shotgun_Siphon_Healing_LowMag") (Variables|/Script/Engine.Actor|GetRootComponent _instigator))))
      (return))))
